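; Camera modifier for the SinisterMelon dive, aiming the camera along the view target's downward axis.
; GetTargetCameraRotation rotates the target actor's forward vector 90 degrees around its right vector and builds the rotation from that X axis alone.
; GetRelativePitchFrustum only passes through the parent's result.
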(fn GetRelativePitchFrustum
  (bind _outrelativepitchfrustum (|CallParentFunction))
  (return _outrelativepitchfrustum))

(fn GetTargetCameraRotation
  (return (Math|Rotator|MakeRotfromX (Math|Vector|RotateVectorAroundAxis (Transformation|GetActorForwardVector (CameraModifier|GetViewTarget)) 9e+01 (Transformation|GetActorRightVector (CameraModifier|GetViewTarget))))))
