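; Camera modifier for the SinisterMelon Mantis that adds bonus field of view while combos are active.
; Combo start and end events maintain a list of active combo indices, and the highest selects a target from a per-combo FOV array.
; Each update interpolates the bonus toward that target scaled by modifier alpha, adds it to FOV and prints a debug value.
; A custom event disables the modifier, waits for its alpha-out time, then removes it from the camera manager.
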(event Custom|DisableAndRemove
  (CameraModifier|DisableModifier false)
  (Utilities|FlowControl|Delay (Variables|CameraModifier|GetAlphaOutTime))
  (Game|Player|RemoveCameraModifier (Variables|CameraModifier|GetCameraOwner) self))

(event Custom|NotifyComboStart
  (params (ComboIndex :type Integer))
  (Utilities|Array|AddUnique (Variables|Transient|GetComboIndexs) ComboIndex))

(event Custom|NotifyComboEnd
  (params (ComboIndex :type Integer))
  (Utilities|Array|RemoveItem (Variables|Transient|GetComboIndexs) ComboIndex))

(fn BlueprintModifyCamera
  (params (DeltaTime :type Float%20%28single-precision%29) (ViewLocation :type Vector :default "0, 0, 0") (ViewRotation :type Rotator :default "0, 0, 0") (FOV :type Float%20%28single-precision%29))
  (bind (_indexofmaxvalue _maxvalue) (Math|Integer|MaxOfIntArray (Variables|Transient|GetComboIndexs)))
  (bind _output_get (Variables|Transient|SetMantisComboBonusFOV (Math|Interpolation|FInterpTo (Variables|Transient|GetMantisComboBonusFOV) (* (Utilities|Array|Get(aref) (Variables|Settings|GetComboAdditionalFovTargetArray) _maxvalue) (Variables|CameraModifier|GetAlpha)) DeltaTime 6.6)))
  (Development|PrintString (Utilities|String|Append "MantisComboBonusFOV: " (Utilities|String|ToString(Float) _output_get)) true false "(R=0.000000,G=0.660000,B=1.000000,A=1.000000)" 2.0 "LemonExamMantisCameraMod")
  (return (Variables|GetViewLocation) (Variables|GetViewRotation) (+ FOV (Variables|Transient|GetMantisComboBonusFOV))))
